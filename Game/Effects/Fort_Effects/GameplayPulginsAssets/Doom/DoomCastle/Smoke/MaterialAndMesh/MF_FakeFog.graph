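# MF_FakeFog is a material function in the Doom Castle smoke effects folder that tints an input color toward a fog color to fake distance fog.
# It divides the actor-to-camera distance by a parameter, applies a power curve, saturates it, and remaps it to a maximum strength as the blend alpha.
# Because it measures from actor position rather than pixel position, the fog amount is uniform across each actor.
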MaterialFunction /Game/Effects/Fort_Effects/GameplayPulginsAssets/Doom/DoomCastle/Smoke/MaterialAndMesh/MF_FakeFog
graph {nodes [
  n1 = Distance(A=ActorPositionWS(), B=CameraPositionWS())
  n2 = ScalarParameter() {defaultValue=100000, group=FakeFog, parameterName=FakeFog_DistanceDivide}
  n3 = Saturate(None=Power(Base=Divide(A=@n1, B=@n2), Exp=ScalarParameter() {defaultValue=2, group=FakeFog, parameterName=FakeFog_FogGradPower}))
  n4 = Constant()
  n5 = Constant() {r=1}
  n6 = ScalarParameter() {defaultValue=0.1, group=FakeFog, parameterName=FakeFog_Strength}
  FunctionOutput(None=LinearInterpolate(
    A=FunctionInput() {inputName=ColorIn}
    B=VectorParameter() {group=FakeFog, parameterName=FakeFog_Color, defaultValue=<rgba 0.2 0.4667 1 1>}.RGB
    Alpha=MaterialFunctionCall(Input=@n3, Input Low=@n4, Input High=@n5, Target Low=@n4, Target High=@n6) {materialFunction=&RemapValueRange}.Result
   ))
 ]}
