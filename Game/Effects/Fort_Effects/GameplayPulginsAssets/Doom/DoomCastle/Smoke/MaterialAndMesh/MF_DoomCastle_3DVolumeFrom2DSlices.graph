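# Material function for Doom Castle smoke effects that fakes a 3D volume texture from a square 2D atlas of slices, outputting RGB and alpha.
# Tiled XY of the UVs input is scaled about each cell's center by Atlas Padding, while Z selects the cell, with atlas U flipped.
# Two samples step through alternating columns, offset by one slice, and a triangle-wave lerp on Z blends them for smooth depth interpolation.
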MaterialFunction /Game/Effects/Fort_Effects/GameplayPulginsAssets/Doom/DoomCastle/Smoke/MaterialAndMesh/MF_DoomCastle_3DVolumeFrom2DSlices
graph {nodes [
  n1 = FunctionInput() {bUsePreviewValueAsDefault=true, inputName=UVs, previewValue=<1 1 1 1>}
  n2 = ComponentMask(None=@n1) {g=true, r=true}
  DDX(Value=@n2)
  DDY(Value=@n2)
  n3 = Add(A=Frac(None=@n2)) {constB=-0.5}
  n4 = FunctionInput() {bUsePreviewValueAsDefault=true, inputName=Atlas Padding, inputType=FunctionInput_Scalar, sortPriority=4, previewValue=<0.877 0.877 0.877 1>}
  n5 = FunctionInput() {bUsePreviewValueAsDefault=true, inputName=Atlas Size, inputType=FunctionInput_Scalar, sortPriority=3, previewValue=<8 8 8 1>}
  n6 = Divide(A=Add(A=Multiply(A=@n3, B=@n4)) {constB=0.5}, B=@n5)
  n7 = ComponentMask(None=@n1) {b=true}
  n8 = Multiply(A=@n5, B=@n5)
  n9 = Multiply(A=@n8) {constB=0.5}
  n10 = Multiply(A=@n7, B=AppendVector(A=@n9, B=@n5)) {constB=8}
  n11 = Floor(None=@n10)
  n12 = AppendVector(A=Multiply(A=@n5) {constB=0.5}, B=@n5)
  n13 = MaterialFunctionCall(Float2=Add(A=@n6, B=Divide(A=@n11, B=@n12)) {constB=0}) {materialFunction=&BreakOutFloat2Components}
  n14 = OneMinus(None=@n13.R)
  n15 = FunctionInput(Preview=TextureObject()) {inputName=Slices Texture, inputType=FunctionInput_Texture2D, sortPriority=1}
  n16 = Subtract(A=@n7, B=Divide(B=@n8))
  n17 = AppendVector(A=@n9, B=@n5)
  n18 = Divide(A=Floor(None=Multiply(A=@n16, B=@n17) {constB=8}), B=@n12)
  n19 = Divide(B=@n5)
  n20 = Constant()
  n21 = MaterialFunctionCall(Float2=Add(A=Add(A=@n6, B=@n18), B=AppendVector(A=@n19, B=@n20))) {materialFunction=&BreakOutFloat2Components}
  n22 = OneMinus(None=@n21.R)
  n23 = Add(A=Frac(None=ComponentMask(None=@n10) {r=true})) {constB=-0.5}
  n24 = LinearInterpolate(
   A=TextureSample(
    UVs=MaterialFunctionCall(X=@n14, Y=@n13.G) {materialFunction=&MakeFloat2}.Result
    Tex=@n15
   ) {samplerType=SAMPLERTYPE_LinearColor}.RGBA
   B=TextureSample(
    UVs=MaterialFunctionCall(X=@n22, Y=@n21.G) {materialFunction=&MakeFloat2}.Result
    Tex=@n15
   ) {samplerType=SAMPLERTYPE_LinearColor}.RGBA
   Alpha=Multiply(A=Abs(None=@n23)) {constB=2}
  )
  FunctionOutput(None=ComponentMask(None=@n24) {b=true, g=true, r=true}) {outputName=RGB}
  FunctionOutput(None=ComponentMask(None=@n24) {a=true}) {outputName=A, sortPriority=1}
 ]}
