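# Material function that decodes first-order spherical harmonics along a direction, stored with the DoomCastle smoke effect assets.
# It normalizes the Direction input and builds a four-term L0/L1 basis: a constant band, then scaled Z, Y and negated X components.
# A dot product of that basis with the four-component SH coefficient input returns a single scalar.
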MaterialFunction /Game/Effects/Fort_Effects/GameplayPulginsAssets/Doom/DoomCastle/Smoke/MaterialAndMesh/MF_SHDecode
graph {nodes [
  n1 = Constant() {r=0.2821}
  n2 = Constant() {r=0.4886}
  n3 = MaterialFunctionCall(Float3=Normalize(VectorInput=FunctionInput() {inputName=Direction})) {materialFunction=&BreakOutFloat3Components}
  n4 = Multiply(A=@n2, B=@n3.B)
  n5 = Multiply(A=@n2, B=@n3.G)
  n6 = Multiply(A=Multiply(A=@n2, B=@n3.R)) {constB=-1}
  FunctionOutput(None=DotProduct(
    A=FunctionInput() {inputName=SH, inputType=FunctionInput_Vector4}
    B=MaterialFunctionCall(X=@n1, Y=@n4, Z=@n5, A=@n6) {materialFunction=&MakeFloat4}.Result
   ))
 ]}
