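# Material function for the Doom Castle smoke cloud body, outputting emissive, opacity, WPO and dithered opacity mask.
# Spherical harmonics from a sliced volume texture, sampled in instance bounds, drive sun, forge and AO shading.
# Time-of-day tint, noise energy and lightning build the emissive, partly hue-shifted by the HeadActive parameter collection value.
# Switches select shadow or camera-facing sprite variants with distance and depth fades, while platform switches trim costlier effects.
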MaterialFunction /Game/Effects/Fort_Effects/GameplayPulginsAssets/Doom/DoomCastle/Smoke/MaterialAndMesh/MF_DoomCloudBody
graph {nodes [
  Constant() {r=1}
  Constant() {r=1}
  n1 = NamedRerouteUsage()
  n2 = VectorParameter() {parameterName=MaskGreen, defaultValue=<rgba 3000 0.6 0.3 0.85>}
  n3 = Multiply(
   A=MaterialFunctionCall(Hue Shift Percentage=@n2.B, Texture=@n1) {materialFunction=&HueShift}.Result
   B=@n2.A
  )
  n4 = OneMinus(None=SphereMask(A=ActorPositionWS(), B=WorldPosition().XYZ, Radius=@n2.R, Hardness=@n2.R))
  n5 = CollectionParameter() {collection=/Game/Effects/Fort_Effects/GameplayPulginsAssets/Doom/DoomCastle/HeadFX/MPC/MPC_LodeSwanHead, parameterName=HeadActive}
  n6 = NamedRerouteUsage()
  n7 = ScalarParameter() {defaultValue=5, parameterName=PowerSunLight}
  n8 = MaterialFunctionCall(
   Input=NamedRerouteUsage()
   Target Low=LinearInterpolate(
    A=ScalarParameter() {defaultValue=0.3, group=Energy, parameterName=EnergyDarkV}
    B=ScalarParameter() {defaultValue=0.3, group=Energy, parameterName=EnergyDarkV_L}
    Alpha=Power(Base=@n6, Exp=@n7)
   )
  ) {materialFunction=&RemapValueRange}
  n9 = Constant3Vector() {constant=<rgba 1 0.6 0.6 1>}
  n10 = Constant3Vector() {constant=<rgba 1 1 1 1>}
  n11 = Constant3Vector() {constant=<rgba 0.8 0.52 0.52 1>}
  n12 = Constant3Vector() {constant=<rgba 0.3 0.6 1.2 1>}
  n13 = Multiply(
   A=NamedRerouteUsage()
   B=MaterialFunctionCall(Input=NamedRerouteUsage(), Target Low=Constant() {r=0.5}) {materialFunction=&RemapValueRange}.Result
  ) {constA=1}
  n14 = Multiply(
   A=MaterialFunctionCall(Hue Shift Percentage=@n2.B, Texture=@n13) {materialFunction=&HueShift}.Result
   B=@n2.A
  )
  n15 = Add(
   A=Multiply(
    A=NamedRerouteUsage()
    B=MaterialFunctionCall(Color_Morning=@n9, Color_Day=@n10, Color_Evening=@n11, Color_Night=@n12) {materialFunction=&Color_Time_of_Day}.Result
   )
   B=LinearInterpolate(A=@n13, B=LinearInterpolate(A=@n13, B=@n14, Alpha=@n4), Alpha=@n5)
  )
  n16 = Add(
   A=LinearInterpolate(A=@n1, B=LinearInterpolate(A=@n1, B=@n3, Alpha=@n4), Alpha=@n5)
   B=Divide(A=Multiply(A=@n8.Result, B=@n15)) {constB=1.97}
  )
  FunctionOutput(None=StaticSwitchParameter(
    True=Constant()
    False=MaterialFunctionCall(Emissive=@n16) {materialFunction=&MF_FX_EmissiveAdjustments}.Result
   ) {parameterName=bIsShadow}) {outputName=Emissive}
  n17 = Constant() {r=1}
  n18 = NamedRerouteUsage()
  n19 = VectorParameter() {
   group=GlobalDither
   parameterName=GlobalDither_EdgeMask_InvertControl
   channelNames {a="", b=Power, g=Multiply, r=Offset}
   defaultValue=<rgba 1 1 1 1>
  }
  n20 = VectorParameter() {
   group=GlobalDither
   parameterName=GlobalDither_EdgeMask_InvertControl_Mobile
   channelNames {a="", b=Power, g=Multiply, r=Offset}
   defaultValue=<rgba 1 1 1 1>
  }
  n21 = VectorParameter() {
   group=GlobalDither
   parameterName=GlobalDither_EdgeMask_Control
   channelNames {a="", b=Power, g=Multiply, r=Offset}
   defaultValue=<rgba 1 1 1 1>
  }
  n22 = MaterialFunctionCall!(
   UorV=NamedRerouteUsage()
   Invert_EdgeControl=MaterialFunctionCall!("Xb1,PS4"=@n19, "PC-High&&Epic,XSX,PS5"=@n19, "PC-Perf,Android"=@n20, Switch=@n20, "PC-Low(NotPerf)&&Medium"=@n19) {materialFunction=/Game/Effects/Niagara/MasterMaterials/Functions/Common/PlatformSwitch_3D}.Result
   NoInvert_EdgeControl=@n21
  ) {materialFunction=/Game/Effects/Niagara/MasterMaterials/Functions/Common/FXCommon_EdgeMask_Ex}
  n23 = OneMinus(None=Power(Base=OneMinus(None=@n22.Mask), Exp=@n19.A))
  n24 = DepthFade(Opacity=Multiply(A=@n18, B=@n23), FadeDistance=ScalarParameter() {defaultValue=1500, parameterName=SpriteDepthFade})
  n25 = NamedRerouteUsage()
  FunctionOutput(None=Multiply(A=StaticSwitchParameter(True=@n17, False=@n24) {parameterName=bIsShadow}, B=@n25)) {outputName=Opacity}
  n26 = NamedRerouteUsage()
  n27 = Constant()
  n28 = StaticSwitchParameter(True=@n27, False=StaticSwitchParameter(True=NamedRerouteUsage(), False=@n27) {parameterName=bIsSprite}) {parameterName=bIsShadow}
  FunctionOutput(None=NamedRerouteDeclaration(None=Add(A=@n26, B=@n28)) {name=FinalWPO, nodeColor=<rgba 0.9453 0 1 1>}) {outputName=WPO}
  n29 = Constant() {r=1}
  n30 = Length(None=Subtract(A=WorldPosition() {worldPositionShaderOffset=WPT_ExcludeAllShaderOffsets}.XYZ, B=CameraPositionWS())) {is=class MaterialExpressionLength}
  n31 = ScalarParameter() {defaultValue=4000, parameterName=SpriteSize_MinDist}
  n32 = Saturate(None=Divide(A=Subtract(A=@n30, B=@n31), B=ScalarParameter() {defaultValue=2500, parameterName=SpriteSize_Distance}))
  n33 = Multiply(A=@n32, B=Multiply(A=@n18, B=@n23))
  n34 = MaterialFunctionCall(Float3=ActorPositionWS()) {materialFunction=&BreakOutFloat3Components}
  n35 = ScalarParameter() {defaultValue=-1000, parameterName=CutOffset}
  n36 = Divide(A=Subtract(A=WorldPosition().Z, B=Add(A=@n34.B, B=@n35)), B=ScalarParameter() {defaultValue=500, parameterName=CutOffsetDivide})
  n37 = Multiply(A=@n25, B=Multiply(A=StaticSwitchParameter(True=@n29, False=@n33) {parameterName=bIsShadow}, B=Saturate(None=@n36)))
  n38 = MaterialFunctionCall(Alpha Threshold=@n37) {materialFunction=&DitherTemporalAA_Fortnite}
  FunctionOutput(None=If(A=@n37, "A > B"=Constant() {r=1}, "A == B"=@n38.Result, "A < B"=@n38.Result) {constB=0.99}) {outputName=Opacity_Mask}
  n39 = MaterialFunctionCall() {materialFunction=&CameraDirectionVector}
  n40 = WorldPosition() {worldPositionShaderOffset=WPT_ExcludeAllShaderOffsets}
  n41 = CameraPositionWS()
  MaterialFunctionCall!(Plane Normal=@n39.Result, Vector=Normalize(VectorInput=Subtract(A=@n40.XYZ, B=@n41))) {materialFunction=/Engine/Functions/Engine_MaterialFunctions02/Math/ProjectVectorOntoPlane}
  Multiply(B=MaterialFunctionCall() {materialFunction=&CameraDirectionVector}.Result) {constA=-1}
  n42 = LinearInterpolate(
   A=VectorParameter() {parameterName=SmokeShading_Color_A, defaultValue=<rgba 0 0 0 1>}.RGB
   B=VectorParameter() {parameterName=SmokeShading_Color_B, defaultValue=<rgba 0.08 0.08 0.08 1>}.RGB
   Alpha=Power(Base=NamedRerouteUsage(), Exp=ScalarParameter() {defaultValue=4, parameterName=SmokeShading_SHResultPower})
  )
  n43 = Add(A=NamedRerouteUsage(), B=ScalarParameter() {parameterName=AO_SHResultOffset})
  n44 = Multiply(
   A=Power(Base=Saturate(None=@n43), Exp=ScalarParameter() {defaultValue=15, parameterName=AO_SHResultPower})
   B=VectorParameter() {parameterName=AO_Color, defaultValue=<rgba 0.55 0.55 0.55 1>}.RGB
  )
  n45 = Constant()
  n46 = Add(
   A=MaterialFunctionCall("Xb1,PS4"=@n44, "PC-High&&Epic,XSX,PS5"=@n44, "PC-Perf,Android"=@n45, Switch=@n45, "PC-Low(NotPerf)&&Medium"=@n44) {materialFunction=&PlatformSwitch_3D}.Result
   B=@n42
  )
  Multiply(A=@n42, B=@n46)
  n47 = Constant4Vector() {constant=<rgba 1.075 -0.32 -0.333 -0.344>}
  n48 = Subtract(A=Constant4Vector() {constant=<rgba 1.579 0.32 0.34 0.287>}.RGBA, B=@n47.RGBA)
  n49 = Add(A=Add(A=WorldPosition() {worldPositionShaderOffset=WPT_ExcludeAllShaderOffsets}.XYZ, B=NamedRerouteUsage()), B=NamedRerouteUsage())
  n50 = Custom!(None) {code="GetPrimitiveData(Parameters).InstanceLocalBoundsCenter.xyz", description=BoundsCenter}
  n51 = Custom() {code="GetPrimitiveData(Parameters).InstanceLocalBoundsExtent.xyz", description=BoundsExtent}
  n52 = Divide(
   A=Subtract(
    A=TransformPosition(None=@n49) {transformSourceType=TRANSFORMPOSSOURCE_World, transformType=TRANSFORMPOSSOURCE_Instance}
    B=Subtract(A=@n50, B=@n51)
   )
   B=Multiply(A=@n51) {constB=2}
  )
  n53 = MaterialFunctionCall(Float3=OneMinus(None=@n52)) {materialFunction=&BreakOutFloat3Components}
  n54 = Subtract(A=MaterialFunctionCall(X=@n53.R, Y=@n53.R, Z=@n53.R) {materialFunction=&MakeFloat3}.Result) {constB=0.5}
  n55 = Constant3Vector() {constant=<rgba 0.9 0.95 0.75 1>}
  n56 = TextureObject() {
   samplerType=SAMPLERTYPE_LinearColor
   texture=&/T_DoomSmokeSH
  }
  n57 = MaterialFunctionCall!(UVs=Add(A=Multiply(A=@n54, B=@n55)) {constB=0.5}, Slices Texture=@n56, Atlas Size=Constant() {r=4}, Atlas Padding=Constant() {r=0.95}) {materialFunction=&/MF_DoomCastle_3DVolumeFrom2DSlices}
  n58 = AppendVector(A=@n57.RGB, B=@n57.RGB)
  NamedRerouteDeclaration(None=Add(A=Multiply(A=@n48, B=@n58), B=@n47.RGBA)) {name=SH_Coefficients, nodeColor=<rgba 0.6641 1 0 1>}
  n59 = NamedRerouteUsage()
  n60 = MaterialFunctionCall!(Direction=Transform(None=SkyAtmosphereLightDirection()) {transformSourceType=TRANSFORMSOURCE_World, transformType=TRANSFORM_Local}, SH=@n59) {materialFunction=&/MF_SHDecode}
  NamedRerouteDeclaration(None=VertexInterpolator(VS=OneMinus(None=@n60.Result)).PS) {name=SHResult_SunLight, nodeColor=<rgba 0 1 0.6953 1>}
  n61 = Add(A=NamedRerouteUsage(), B=WorldPosition() {worldPositionShaderOffset=WPT_ExcludeAllShaderOffsets}.XYZ)
  n62 = Add(A=ActorPositionWS(), B=Constant3Vector() {constant=<rgba -4500 1500 -2500 1>})
  n63 = Transform(None=Normalize(VectorInput=Subtract(A=@n61, B=@n62))) {transformSourceType=TRANSFORMSOURCE_World, transformType=TRANSFORM_Local}
  NamedRerouteDeclaration(None=VertexInterpolator(VS=MaterialFunctionCall(Direction=@n63, SH=@n59) {materialFunction=&MF_SHDecode}.Result).PS) {name=SHResult_ForgeLight, nodeColor=<rgba 0 1 0.6953 1>}
  n64 = OneMinus(None=NamedRerouteUsage())
  n65 = Constant() {r=-0.04}
  n66 = Saturate(None=Multiply(A=Add(A=@n64, B=@n65), B=Constant() {r=3}))
  n67 = Constant() {r=10}
  n68 = Add(A=WorldPosition() {worldPositionShaderOffset=WPT_ExcludeAllShaderOffsets}.XYZ, B=NamedRerouteUsage())
  n69 = Sine(None=Multiply(A=Time(), B=Constant() {r=0.0006}))
  n70 = Constant() {r=114}
  n71 = Abs(None=Sine(None=Multiply(A=@n69, B=@n70)))
  n72 = Constant() {r=0.3}
  n73 = MaterialFunctionCall(
   X=Multiply(
    A=MaterialFunctionCall(Input=@n71, Target Low=@n72) {materialFunction=&RemapValueRange}.Result
    B=Constant() {r=7000}
   ) {constB=6000}
   Y=Constant()
   Z=Constant() {r=4500}
  ) {materialFunction=&MakeFloat3}
  n74 = ActorPositionWS()
  n75 = VertexInterpolator(VS=Subtract(A=@n68, B=Add(A=@n73.Result, B=@n74)))
  n76 = Constant3Vector() {constant=<rgba 1 1 2 1>}
  n77 = Divide(A=Length(None=Multiply(A=@n75.PS, B=@n76)), B=Constant() {r=7000})
  n78 = NamedRerouteUsage()
  n79 = Constant() {r=0.5}
  n80 = Multiply(A=Time(), B=Constant() {r=0.2})
  n81 = Constant() {r=0.8}
  n82 = Saturate(None=Subtract(A=Sine(None=@n80), B=@n81))
  n83 = Divide(B=OneMinus(None=@n81))
  n84 = Multiply(
   A=OneMinus(None=Saturate(None=@n77))
   B=Multiply(
    A=MaterialFunctionCall(Input=@n78, Target Low=@n79) {materialFunction=&RemapValueRange}.Result
    B=Multiply(A=@n82, B=@n83) {constB=10}
   )
  )
  n85 = Constant() {r=4}
  n86 = VectorParameter() {parameterName=LightningBoltC, defaultValue=<rgba 150 0 0 0.05>}
  n87 = VectorParameter() {parameterName=CameraBaseReduceFlick, defaultValue=<rgba 1000 1500 0 1>}
  n88 = Subtract(A=Distance(A=@n68, B=CameraPositionWS()), B=@n87.R)
  n89 = LinearInterpolate(A=Multiply(A=@n86.RGB, B=@n86.RGB), B=@n86.RGB, Alpha=Saturate(None=Divide(A=@n88, B=@n87.G)))
  n90 = Multiply(A=Saturate(None=Power(Base=@n66, Exp=@n67)), B=Multiply(A=Power(Base=@n84, Exp=@n85), B=VertexInterpolator(VS=@n89).PS))
  n91 = Constant()
  NamedRerouteDeclaration(None=MaterialFunctionCall("Xb1,PS4"=@n90, "PC-High&&Epic,XSX,PS5"=@n90, "PC-Perf,Android"=@n91, Switch=@n91, "PC-Low(NotPerf)&&Medium"=@n90) {materialFunction=&PlatformSwitch_3D}.Result) {name=LightningFX, nodeColor=<rgba 0.9453 1 0 1>}
  n92 = TextureObject() {texture=/Game/Environments/Asteria/Sets/NeonCity/Techart/Screen/Textures/Random_M}
  n93 = Constant3Vector() {constant=<rgba 1 1 1 0>}
  n94 = Constant()
  n95 = MaterialFunctionCall(R=Constant(), G=Constant() {r=0.5}) {materialFunction=&AppendMany}
  NamedRerouteDeclaration(None=LinearInterpolate(
    B=MaterialFunctionCall(Texture=@n92, UVControl_RGB=@n93, UVControl_A=@n94, "[Override]UVInput"=@n95.RG) {materialFunction=&FXCommon_TextureSample}.R
    Alpha=ScalarParameter() {defaultValue=1, parameterName=RandomV Switch}
   ) {constAlpha=1}) {name=RandomV, nodeColor=<rgba 1 0.6562 0 1>}
  n96 = VectorParameter() {parameterName=DetailTexture_RotAxis, defaultValue=<rgba 0 1 0 0.1>}
  n97 = Constant3Vector()
  n98 = MaterialFunctionCall(Float3=@n52) {materialFunction=&BreakOutFloat3Components}
  n99 = VectorParameter() {parameterName=DetailUVGrad, defaultValue=<rgba 0 2 0.8 1>}
  n100 = Saturate(None=Multiply(A=Add(A=@n98.R, B=@n99.R), B=@n99.G))
  n101 = Multiply(
   A=Add(A=@n52, B=RotateAboutAxis(NormalizedRotationAxis=@n96.RGB, RotationAngle=@n96.RGB, PivotPoint=@n97, Position=@n52))
   B=LinearInterpolate(
    A=VectorParameter() {parameterName=ScaleDetailUV2, defaultValue=<rgba 3.5 3 3.5 1>}.RGB
    B=VectorParameter() {parameterName=ScaleDetailUV, defaultValue=<rgba 0.5 1 1 1>}.RGB
    Alpha=Power(Base=@n100, Exp=@n99.B)
   )
  )
  n102 = Multiply(A=VectorParameter() {parameterName=DetailMove, defaultValue=<rgba -0.01 0 0 1>}.RGB, B=Time() {bOverride_Period=true, period=6666})
  n103 = MaterialFunctionCall!(
   UVs=VertexInterpolator(VS=Add(A=@n101, B=@n102)).PS
   Slices Texture=TextureObject() {samplerType=SAMPLERTYPE_LinearColor, texture=/Game/Environments/Helios/TechArt/LowAltitudeClouds/T_LowAltitudeClouds_N}
   Atlas Size=Constant() {r=8}
   Atlas Padding=ScalarParameter() {defaultValue=0.935, parameterName=AtlasPadding, primitiveDataIndex=8}
  ) {materialFunction=/Game/Environments/Helios/TechArt/LowAltitudeClouds/MF_3DVolumeFrom2DSlices}
  n104 = ScalarParameter() {defaultValue=2, group=Energy, parameterName=Energy_Power}
  NamedRerouteDeclaration(None=Saturate(None=Power(Base=@n103.A, Exp=@n104))) {name=EnergySampleResult, nodeColor=<rgba 0.3359 0 1 1>}
  n105 = Constant()
  n106 = NamedRerouteUsage()
  n107 = Saturate(None=Multiply(A=Add(A=@n106, B=@n105), B=Constant() {r=2}))
  n108 = Constant() {r=4}
  n109 = MaterialFunctionCall(In=Saturate(None=Power(Base=@n107, Exp=@n108)), Contrast=Constant() {r=0.2}) {materialFunction=&CheapContrast}
  n110 = Constant3Vector() {constant=<rgba 0.05 0 0 1>}
  n111 = NamedRerouteUsage()
  n112 = Constant()
  n113 = Saturate(None=Multiply(A=Add(A=@n106, B=@n112), B=Constant() {r=2}))
  n114 = LinearInterpolate(A=Constant() {r=20}, B=Constant() {r=10}, Alpha=NamedRerouteUsage())
  n115 = Multiply(A=Saturate(None=Power(Base=@n113, Exp=@n114)), B=VectorParameter() {parameterName=Flicker Color, defaultValue=<rgba 300 0 0 1>}.RGB)
  NamedRerouteDeclaration(None=Add(A=Multiply(A=@n109.Result, B=@n110), B=Multiply(A=@n111, B=@n115))) {name=ForgeLight, nodeColor=<rgba 0 1 0.0156 1>}
  n116 = NamedRerouteUsage()
  n117 = Add(
   A=Multiply(A=@n116, B=ScalarParameter() {defaultValue=2, parameterName=SineFreq})
   B=Multiply(A=ScalarParameter() {defaultValue=-0.2, parameterName=SineMove}, B=Time() {bOverride_Period=true, period=6666})
  )
  n118 = Multiply(A=ConstantBiasScale(None=Sine(None=@n117))) {constB=-1}
  n119 = Normalize(VectorInput=VertexNormalWS())
  n120 = Multiply(A=Multiply(A=Multiply(A=@n118, B=@n119), B=@n116), B=ScalarParameter() {defaultValue=500, parameterName=SineSTR})
  n121 = MaterialFunctionCall() {materialFunction=&CameraDirectionVector}
  n122 = Subtract(A=@n120, B=Multiply(A=DotProduct(A=@n120, B=@n121.Result), B=@n121.Result))
  n123 = Constant()
  NamedRerouteDeclaration(None=MaterialFunctionCall("Xb1,PS4"=@n122, "PC-High&&Epic,XSX,PS5"=@n122, "PC-Perf,Android"=@n123, Switch=@n123, "PC-Low(NotPerf)&&Medium"=@n122) {materialFunction=&PlatformSwitch_3D}.Result) {name=SineWPO, nodeColor=<rgba 0 1 0.7656 1>}
  n124 = WorldPosition()
  n125 = Custom() {code="GetPrimitiveData(Parameters).InstanceLocalBoundsCenter.xyz", description=BoundsCenter}
  n126 = Custom() {code="GetPrimitiveData(Parameters).InstanceLocalBoundsExtent.xyz", description=BoundsExtent}
  n127 = Divide(
   A=Subtract(
    A=TransformPosition(None=@n124.XYZ) {transformSourceType=TRANSFORMPOSSOURCE_World, transformType=TRANSFORMPOSSOURCE_Instance}
    B=Subtract(A=@n125, B=@n126)
   )
   B=Multiply(A=@n126) {constB=2}
  )
  n128 = RotateAboutAxis(
   NormalizedRotationAxis=Constant3Vector() {constant=<rgba 0 1 0 0>}
   RotationAngle=Constant() {r=-0.05}
   PivotPoint=Constant3Vector()
   Position=@n127
  )
  NamedRerouteDeclaration(None=MaterialFunctionCall(Float3=Add(A=@n127, B=@n128)) {materialFunction=&BreakOutFloat3Components}.B) {name=Rotated Bound Z, nodeColor=<rgba 1 0.4453 0 1>}
  NamedRerouteDeclaration(None=@n127) {name=InstanceBound, nodeColor=<rgba 0.125 0 1 1>}
  n129 = ScalarParameter() {defaultValue=0.4, parameterName=ScaleO}
  n130 = CameraPositionWS()
  n131 = VectorParameter() {parameterName=ForgePosition, defaultValue=<rgba 63638 88679 9900 1>}
  n132 = Length(None=Multiply(A=Subtract(A=@n130, B=@n131.RGB), B=VectorParameter() {parameterName=ScaleInforgeRange, defaultValue=<rgba 1 1 0.6 1>}.RGB))
  n133 = VectorParameter() {parameterName=InforgeScaleParam, defaultValue=<rgba 2100 2000 7000 2000>}
  n134 = Saturate(None=Divide(A=Subtract(A=@n132, B=@n133.R), B=@n133.G))
  n135 = WorldPosition()
  n136 = ActorPositionWS()
  n137 = Divide(A=Subtract(A=Distance(A=@n135.XYZ, B=@n136), B=@n133.B), B=@n133.A)
  NamedRerouteDeclaration(None=LinearInterpolate(A=LinearInterpolate(A=@n129, Alpha=@n134) {constA=1}, Alpha=Saturate(None=@n137))) {name=InforgeScaleO, nodeColor=<rgba 1 0.3281 0 1>}
  n138 = Subtract(A=@n61, B=CameraPositionWS())
  n139 = MaterialFunctionCall(Direction=Transform(None=Normalize(VectorInput=@n138)) {transformSourceType=TRANSFORMSOURCE_World, transformType=TRANSFORM_Local}, SH=@n59) {materialFunction=&MF_SHDecode}
  NamedRerouteDeclaration(None=VertexInterpolator(VS=OneMinus(None=@n139.Result)).PS) {name=SHResult_CameraVector, nodeColor=<rgba 0 1 0.6953 1>}
  n140 = Subtract(A=@n61, B=Add(A=ActorPositionWS(), B=VectorParameter() {parameterName=AO_Pos, defaultValue=<rgba 0 0 35000 1>}.RGB))
  n141 = MaterialFunctionCall(Direction=Transform(None=Normalize(VectorInput=@n140)) {transformSourceType=TRANSFORMSOURCE_World, transformType=TRANSFORM_Local}, SH=@n59) {materialFunction=&MF_SHDecode}
  NamedRerouteDeclaration(None=VertexInterpolator(VS=OneMinus(None=@n141.Result)).PS) {name=SHResult_GI, nodeColor=<rgba 0 1 0.6953 1>}
  NamedRerouteDeclaration(None=PixelNormalWS()) {name=RealPixelNormal, nodeColor=<rgba 1 0 0.9375 1>}
  n142 = Normalize(VectorInput=WorldPosition() {worldPositionShaderOffset=WPT_CameraRelativeNoOffsets}.XYZ)
  n143 = ScalarParameter() {defaultValue=1500, parameterName=SpriteCameraOffset}
  n144 = TextureCoordinate()
  n145 = AppendVector(A=Multiply(A=Add(A=@n144) {constB=-0.5}) {constB=2}, B=Constant())
  n146 = RotateAboutAxis(
   NormalizedRotationAxis=Constant3Vector() {constant=<rgba 0 0 1 0>}
   RotationAngle=Constant() {r=0.125}
   PivotPoint=Constant3Vector()
   Position=@n145
  )
  n147 = Normalize(VectorInput=@n39.Result)
  n148 = CrossProduct(A=Transform(None=Constant3Vector() {constant=<rgba 0 1 0 0>}) {transformSourceType=TRANSFORMSOURCE_Camera}, B=@n147)
  n149 = ScalarParameter() {defaultValue=1500, parameterName=SpriteSize}
  n150 = StaticSwitchParameter(
   True=MaterialFunctionCall("Xb1,PS4"=@n149, "PC-High&&Epic,XSX,PS5"=@n149, "PC-Perf,Android"=Constant(), Switch=@n149, "PC-Low(NotPerf)&&Medium"=@n149) {materialFunction=&PlatformSwitch_1D}.Result
   False=@n149
  ) {parameterName=UseTranslucentScalability}
  n151 = ScalarParameter() {defaultValue=-0.3, group=GlobalDither, parameterName=GlobalDitherToSpriteSize_InvertOffset_XB1PS4}
  n152 = VectorParameter() {
   group=GlobalDither
   parameterName=GlobalDitherToSpriteSize_EdgeMask_InvertControl
   channelNames {a="", b=Power, g=Multiply, r=Offset}
   defaultValue=<rgba 1 1 1 1>
  }
  n153 = VectorParameter() {
   group=GlobalDither
   parameterName=GlobalDitherToSpriteSize_EdgeMask_InvertControl_Mobile
   channelNames {a="", b=Power, g=Multiply, r=Offset}
   defaultValue=<rgba 1 1 1 1>
  }
  n154 = MaterialFunctionCall(Float3=MaterialFunctionCall("Xb1,PS4"=@n152, "PC-High&&Epic,XSX,PS5"=@n152, "PC-Perf,Android"=@n153, Switch=@n153, "PC-Low(NotPerf)&&Medium"=@n152) {materialFunction=&PlatformSwitch_3D}.Result) {materialFunction=&BreakOutFloat3Components}
  n155 = ScalarParameter() {group=GlobalDither, parameterName=GlobalDitherToSpriteSize_InvertOffset_Mobile}
  n156 = ScalarParameter() {defaultValue=-0.15, group=GlobalDither, parameterName=GlobalDitherToSpriteSize_InvertOffset_LOWMED}
  n157 = VectorParameter() {
   group=GlobalDither
   parameterName=GlobalDitherToSpriteSize_EdgeMask_Control
   channelNames {a="", b=Power, g=Multiply, r=Offset}
   defaultValue=<rgba 1 1 1 1>
  }
  n158 = MaterialFunctionCall(
   UorV=NamedRerouteUsage()
   Invert_EdgeControl=MaterialFunctionCall(
    X=MaterialFunctionCall("Xb1,PS4"=@n151, "PC-High&&Epic,XSX,PS5"=@n154.R, "PC-Perf,Android"=@n155, Switch=@n155, "PC-Low(NotPerf)&&Medium"=@n156) {materialFunction=&PlatformSwitch_1D}.Result
    Y=@n154.G
    Z=@n154.G
   ) {materialFunction=&MakeFloat3}.Result
   NoInvert_EdgeControl=@n157
  ) {materialFunction=&FXCommon_EdgeMask_Ex}
  n159 = OneMinus(None=Power(Base=OneMinus(None=@n158.Mask), Exp=@n152.A))
  n160 = Multiply(
   A=MaterialFunctionCall(VectorToTransform=Add(A=@n146, B=@n145), BasisX=@n148, BasisY=CrossProduct(A=@n147, B=@n148), BasisZ=@n147) {materialFunction=&Transform3x3Matrix}.Transformed Vector
   B=Multiply(A=@n32, B=Multiply(A=@n150, B=@n159))
  )
  NamedRerouteDeclaration(None=Add(A=Multiply(A=@n142, B=@n143), B=@n160)) {name=SpriteVersionWPO, nodeColor=<rgba 0 1 0.5312 1>}
  n161 = TextureSample() {texture=&/SmokeMask_02_T2}
  n162 = Constant()
  n163 = ScalarParameter() {defaultValue=0.02, parameterName=OpacityOffset_Mobile}
  n164 = MaterialFunctionCall("Xb1,PS4"=@n162, "PC-High&&Epic,XSX,PS5"=@n162, "PC-Perf,Android"=@n163, Switch=@n163, "PC-Low(NotPerf)&&Medium"=@n162) {materialFunction=&PlatformSwitch_1D}
  n165 = ScalarParameter() {defaultValue=1, parameterName=OpacityPower}
  n166 = ScalarParameter() {defaultValue=0.7, parameterName=OpacityPower_Mobile}
  NamedRerouteDeclaration(None=Power(
    Base=Subtract(A=@n161.R, B=@n164.Result)
    Exp=MaterialFunctionCall("Xb1,PS4"=@n165, "PC-High&&Epic,XSX,PS5"=@n165, "PC-Perf,Android"=@n166, Switch=@n166, "PC-Low(NotPerf)&&Medium"=@n165) {materialFunction=&PlatformSwitch_1D}.Result
   )) {name=SpriteShapeMask, nodeColor=<rgba 0.2891 0 1 1>}
  n167 = DotProduct(A=SkyAtmosphereLightDirection(), B=VertexNormalWS())
  n168 = ScalarParameter() {defaultValue=0.7, parameterName=DirectLighting_Offset}
  n169 = Saturate(None=Divide(A=Add(A=@n167, B=@n168), B=Add(A=@n168)))
  n170 = ScalarParameter() {defaultValue=0.5, parameterName=DirectLighting_Power}
  n171 = VertexInterpolator(VS=Multiply(A=Power(Base=@n169, Exp=@n170), B=VectorParameter() {parameterName=LightC, defaultValue=<rgba 0.01 0.01 0.01 1>}.RGB))
  n172 = Constant()
  n173 = Multiply(A=DotProduct(A=CameraVectorWS(), B=SkyAtmosphereLightDirection())) {constB=-1}
  n174 = NamedRerouteUsage()
  n175 = ScalarParameter() {defaultValue=9, parameterName=Transmitance_SHResultPower}
  n176 = Multiply(
   A=Power(Base=Saturate(None=@n173), Exp=ScalarParameter() {defaultValue=5, parameterName=Transmitance_AreaPower})
   B=Multiply(A=VectorParameter() {parameterName=Transmitance_Color, defaultValue=<rgba 10 0 0 1>}.RGB, B=Power(Base=@n174, Exp=@n175))
  ) {constA=4, constB=4}
  n177 = Constant()
  n178 = MaterialFunctionCall("Xb1,PS4"=@n176, "PC-High&&Epic,XSX,PS5"=@n176, "PC-Perf,Android"=@n177, Switch=@n177, "PC-Low(NotPerf)&&Medium"=@n176) {materialFunction=&PlatformSwitch_3D}
  NamedRerouteDeclaration(None=Add(
    A=MaterialFunctionCall("Xb1,PS4"=@n171.PS, "PC-High&&Epic,XSX,PS5"=@n171.PS, "PC-Perf,Android"=@n172, Switch=@n172, "PC-Low(NotPerf)&&Medium"=@n171.PS) {materialFunction=&PlatformSwitch_3D}.Result
    B=Add(A=@n178.Result, B=@n46)
   )) {name=Shading, nodeColor=<rgba 1 0 0.7734 1>}
  NamedRerouteDeclaration(None=@n160) {name=SpriteWPOWithoutNearFade, nodeColor=<rgba 1 0 0.9141 1>}
  n179 = Constant() {r=14000}
  n180 = Constant()
  NamedRerouteDeclaration(None=Power(Base=SphereMask(A=@n61, B=@n62, Radius=@n179, Hardness=@n180), Exp=Constant() {r=5})) {name=ForgeLight_SphereMask, nodeColor=<rgba 1 0 0.1632 1>}
  NamedRerouteUsage()
  NamedRerouteUsage()
  NamedRerouteUsage()
 ]}
