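# M_DoomCastleSmoke is an unlit translucent material, by its name and folder the smoke effect around Doom Castle.
# The MF_DoomCloudBody material function holds the cloud logic, with its Opacity and Opacity Mask outputs passed to reroutes.
# A Scalar_Time_of_Day function node is also present, with no visible connections.
# Dithered LOD transitions are enabled, and automatic usage setting in the editor is off.
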Material /Game/Effects/Fort_Effects/GameplayPulginsAssets/Doom/DoomCastle/Smoke/MaterialAndMesh/M_DoomCastleSmoke
graph {
 properties {
  bAutomaticallySetUsageInEditor=false
  blendMode=BLEND_Translucent
  ditheredLODTransition=true
  shadingModel=MSM_Unlit
  usedShadingModels=Unlit
 }
 nodes [
  MaterialFunctionCall() {materialFunction=&Scalar_Time_of_Day}
  n1 = MaterialFunctionCall() {materialFunction=&/MF_DoomCloudBody}
  Reroute(None=@n1.Opacity)
  Reroute(None=@n1.Opacity_Mask)
 ]
}
