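# Mobile material function for the DoomCastle smoke cloud body, outputting Emissive, Opacity and WPO with static switches for shadow and sprite variants.
# Spherical harmonics from a sliced volume atlas are decoded against sun direction per vertex, driving shading tinted by time of day.
# WPO expands camera-facing sprites, scaled by camera distance and edge mask.
# Opacity uses smoke mask, depth fade and global-dither edge mask, with forge-proximity scaling also computed.
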MaterialFunction /Game/Effects/Fort_Effects/GameplayPulginsAssets/Doom/DoomCastle/Smoke/MaterialAndMesh/MF_DoomCloudBody_Mobile
graph {nodes [
  Constant() {r=1}
  Constant() {r=1}
  n1 = Multiply(
   A=NamedRerouteUsage()
   B=MaterialFunctionCall(
    Color_Morning=Constant3Vector() {constant=<rgba 1 0.6 0.6 1>}
    Color_Day=Constant3Vector() {constant=<rgba 1 1 1 1>}
    Color_Evening=Constant3Vector() {constant=<rgba 0.8 0.52 0.52 1>}
    Color_Night=Constant3Vector() {constant=<rgba 0.3 0.6 1.2 1>}
   ) {materialFunction=&Color_Time_of_Day}.Result
  )
  FunctionOutput(None=StaticSwitchParameter(
    True=Constant()
    False=MaterialFunctionCall(Emissive=@n1) {materialFunction=&MF_FX_EmissiveAdjustments}.Result
   ) {parameterName=bIsShadow}) {outputName=Emissive}
  n2 = Constant() {r=1}
  n3 = NamedRerouteUsage()
  n4 = VectorParameter() {
   group=GlobalDither
   parameterName=GlobalDither_EdgeMask_InvertControl_Mobile
   channelNames {a="", b=Power, g=Multiply, r=Offset}
   defaultValue=<rgba 1 1 1 1>
  }
  n5 = VectorParameter() {
   group=GlobalDither
   parameterName=GlobalDither_EdgeMask_Control
   channelNames {a="", b=Power, g=Multiply, r=Offset}
   defaultValue=<rgba 1 1 1 1>
  }
  n6 = Power(
   Base=OneMinus(None=MaterialFunctionCall!(UorV=@n3, Invert_EdgeControl=@n4, NoInvert_EdgeControl=@n5) {materialFunction=/Game/Effects/Niagara/MasterMaterials/Functions/Common/FXCommon_EdgeMask_Ex}.Mask)
   Exp=@n4.A
  )
  n7 = DepthFade(
   Opacity=Multiply(A=NamedRerouteUsage(), B=OneMinus(None=@n6))
   FadeDistance=ScalarParameter() {defaultValue=1500, parameterName=SpriteDepthFade}
  )
  FunctionOutput(None=Multiply(A=StaticSwitchParameter(True=@n2, False=@n7) {parameterName=bIsShadow}, B=NamedRerouteUsage())) {outputName=Opacity}
  n8 = Constant()
  n9 = StaticSwitchParameter(True=NamedRerouteUsage(), False=@n8) {parameterName=bIsSprite}
  FunctionOutput(None=NamedRerouteDeclaration(None=StaticSwitchParameter(True=@n8, False=@n9) {parameterName=bIsShadow}) {name=FinalWPO, nodeColor=<rgba 0.9453 0 1 1>}) {outputName=WPO}
  n10 = MaterialFunctionCall() {materialFunction=&CameraDirectionVector}
  n11 = WorldPosition() {worldPositionShaderOffset=WPT_ExcludeAllShaderOffsets}
  n12 = CameraPositionWS()
  MaterialFunctionCall!(Plane Normal=@n10.Result, Vector=Normalize(VectorInput=Subtract(A=@n11.XYZ, B=@n12))) {materialFunction=/Engine/Functions/Engine_MaterialFunctions02/Math/ProjectVectorOntoPlane}
  n13 = Constant4Vector() {constant=<rgba 1.075 -0.32 -0.333 -0.344>}
  n14 = Subtract(A=Constant4Vector() {constant=<rgba 1.579 0.32 0.34 0.287>}.RGBA, B=@n13.RGBA)
  n15 = Add(A=WorldPosition() {worldPositionShaderOffset=WPT_ExcludeAllShaderOffsets}.XYZ, B=NamedRerouteUsage())
  n16 = NamedRerouteUsage()
  n17 = Custom!(None) {code="GetPrimitiveData(Parameters).InstanceLocalBoundsExtent.xyz", description=BoundsExtent}
  n18 = Subtract(
   A=TransformPosition(None=Add(A=@n15, B=@n16)) {transformSourceType=TRANSFORMPOSSOURCE_World, transformType=TRANSFORMPOSSOURCE_Instance}
   B=Subtract(A=Custom() {code="GetPrimitiveData(Parameters).InstanceLocalBoundsCenter.xyz", description=BoundsCenter}, B=@n17)
  )
  n19 = Multiply(A=@n17) {constB=2}
  n20 = MaterialFunctionCall(Float3=OneMinus(None=Divide(A=@n18, B=@n19))) {materialFunction=&BreakOutFloat3Components}
  n21 = Subtract(A=MaterialFunctionCall(X=@n20.R, Y=@n20.R, Z=@n20.R) {materialFunction=&MakeFloat3}.Result) {constB=0.5}
  n22 = Constant3Vector() {constant=<rgba 0.9 0.95 0.75 1>}
  n23 = TextureObject() {
   samplerType=SAMPLERTYPE_LinearColor
   texture=&/T_DoomSmokeSH
  }
  n24 = MaterialFunctionCall!(UVs=Add(A=Multiply(A=@n21, B=@n22)) {constB=0.5}, Slices Texture=@n23, Atlas Size=Constant() {r=4}, Atlas Padding=Constant() {r=0.95}) {materialFunction=&/MF_DoomCastle_3DVolumeFrom2DSlices}
  n25 = AppendVector(A=@n24.RGB, B=@n24.RGB)
  NamedRerouteDeclaration(None=Add(A=Multiply(A=@n14, B=@n25), B=@n13.RGBA)) {name=SH_Coefficients, nodeColor=<rgba 0.6641 1 0 1>}
  n26 = MaterialFunctionCall!(
   Direction=Transform(None=SkyAtmosphereLightDirection()) {transformSourceType=TRANSFORMSOURCE_World, transformType=TRANSFORM_Local}
   SH=NamedRerouteUsage()
  ) {materialFunction=&/MF_SHDecode}
  NamedRerouteDeclaration(None=VertexInterpolator(VS=OneMinus(None=@n26.Result)).PS) {name=SHResult_SunLight, nodeColor=<rgba 0 1 0.6953 1>}
  n27 = WorldPosition()
  n28 = Custom() {code="GetPrimitiveData(Parameters).InstanceLocalBoundsCenter.xyz", description=BoundsCenter}
  n29 = Custom() {code="GetPrimitiveData(Parameters).InstanceLocalBoundsExtent.xyz", description=BoundsExtent}
  n30 = Divide(
   A=Subtract(
    A=TransformPosition(None=@n27.XYZ) {transformSourceType=TRANSFORMPOSSOURCE_World, transformType=TRANSFORMPOSSOURCE_Instance}
    B=Subtract(A=@n28, B=@n29)
   )
   B=Multiply(A=@n29) {constB=2}
  )
  n31 = RotateAboutAxis(
   NormalizedRotationAxis=Constant3Vector() {constant=<rgba 0 1 0 0>}
   RotationAngle=Constant() {r=-0.05}
   PivotPoint=Constant3Vector()
   Position=@n30
  )
  NamedRerouteDeclaration(None=MaterialFunctionCall(Float3=Add(A=@n30, B=@n31)) {materialFunction=&BreakOutFloat3Components}.B) {name=Rotated Bound Z, nodeColor=<rgba 1 0.4453 0 1>}
  NamedRerouteDeclaration(None=@n30) {name=InstanceBound, nodeColor=<rgba 0.125 0 1 1>}
  n32 = ScalarParameter() {defaultValue=0.4, parameterName=ScaleO}
  n33 = CameraPositionWS()
  n34 = VectorParameter() {parameterName=ForgePosition, defaultValue=<rgba 63638 88679 9900 1>}
  n35 = Length(None=Multiply(A=Subtract(A=@n33, B=@n34.RGB), B=VectorParameter() {parameterName=ScaleInforgeRange, defaultValue=<rgba 1 1 0.6 1>}.RGB)) {is=class MaterialExpressionLength}
  n36 = VectorParameter() {parameterName=InforgeScaleParam, defaultValue=<rgba 2100 2000 7000 2000>}
  n37 = Saturate(None=Divide(A=Subtract(A=@n35, B=@n36.R), B=@n36.G))
  n38 = WorldPosition()
  n39 = ActorPositionWS()
  n40 = Divide(A=Subtract(A=Distance(A=@n38.XYZ, B=@n39), B=@n36.B), B=@n36.A)
  NamedRerouteDeclaration(None=LinearInterpolate(A=LinearInterpolate(A=@n32, Alpha=@n37) {constA=1}, Alpha=Saturate(None=@n40))) {name=InforgeScaleO, nodeColor=<rgba 1 0.3281 0 1>}
  NamedRerouteDeclaration(None=PixelNormalWS()) {name=RealPixelNormal, nodeColor=<rgba 1 0 0.9375 1>}
  n41 = Normalize(VectorInput=WorldPosition() {worldPositionShaderOffset=WPT_CameraRelativeNoOffsets}.XYZ)
  n42 = ScalarParameter() {defaultValue=1500, parameterName=SpriteCameraOffset}
  n43 = TextureCoordinate()
  n44 = AppendVector(A=Multiply(A=Add(A=@n43) {constB=-0.5}) {constB=2}, B=Constant())
  n45 = RotateAboutAxis(
   NormalizedRotationAxis=Constant3Vector() {constant=<rgba 0 0 1 0>}
   RotationAngle=Constant() {r=0.125}
   PivotPoint=Constant3Vector()
   Position=@n44
  )
  n46 = Normalize(VectorInput=@n10.Result)
  n47 = CrossProduct(A=Transform(None=Constant3Vector() {constant=<rgba 0 1 0 0>}) {transformSourceType=TRANSFORMSOURCE_Camera}, B=@n46)
  n48 = Subtract(A=WorldPosition() {worldPositionShaderOffset=WPT_ExcludeAllShaderOffsets}.XYZ, B=CameraPositionWS())
  n49 = Divide(
   A=Subtract(A=Length(None=@n48), B=ScalarParameter() {defaultValue=4000, parameterName=SpriteSize_MinDist})
   B=ScalarParameter() {defaultValue=2500, parameterName=SpriteSize_Distance}
  )
  n50 = ScalarParameter() {defaultValue=3000, parameterName=SpriteSize_Mobile}
  n51 = VectorParameter() {
   group=GlobalDither
   parameterName=GlobalDitherToSpriteSize_EdgeMask_InvertControl_Mobile
   channelNames {a="", b=Power, g=Multiply, r=Offset}
   defaultValue=<rgba 1 1 1 1>
  }
  n52 = VectorParameter() {
   group=GlobalDither
   parameterName=GlobalDitherToSpriteSize_EdgeMask_Control
   channelNames {a="", b=Power, g=Multiply, r=Offset}
   defaultValue=<rgba 1 1 1 1>
  }
  n53 = MaterialFunctionCall(
   UorV=NamedRerouteUsage()
   Invert_EdgeControl=MaterialFunctionCall(X=@n51.Offset, Y=@n51.Offset, Z=@n51.Offset) {materialFunction=&MakeFloat3}.Result
   NoInvert_EdgeControl=@n52
  ) {materialFunction=&FXCommon_EdgeMask_Ex}
  n54 = OneMinus(None=Power(Base=OneMinus(None=@n53.Mask), Exp=@n51.A))
  n55 = Multiply(
   A=MaterialFunctionCall(VectorToTransform=Add(A=@n45, B=@n44), BasisX=@n47, BasisY=CrossProduct(A=@n46, B=@n47), BasisZ=@n46) {materialFunction=&Transform3x3Matrix}.Transformed Vector
   B=Multiply(A=Saturate(None=@n49), B=Multiply(A=@n50, B=@n54))
  )
  NamedRerouteDeclaration(None=Add(A=Multiply(A=@n41, B=@n42), B=@n55)) {name=SpriteVersionWPO, nodeColor=<rgba 0 1 0.5312 1>}
  NamedRerouteDeclaration(None=Power(
    Base=TextureSample() {texture=&/SmokeMask_02_T2}.R
    Exp=ScalarParameter() {defaultValue=1, parameterName=OpacityPower}
   )) {name=SpriteShapeMask, nodeColor=<rgba 0.2891 0 1 1>}
  n56 = VectorParameter() {parameterName=SmokeShading_Color_A, defaultValue=<rgba 0 0 0 1>}
  n57 = VectorParameter() {parameterName=SmokeShading_Color_B, defaultValue=<rgba 0.08 0.08 0.08 1>}
  n58 = Power(Base=NamedRerouteUsage(), Exp=ScalarParameter() {defaultValue=4, parameterName=SmokeShading_SHResultPower})
  NamedRerouteDeclaration(None=Multiply(A=ScalarParameter() {defaultValue=0.45, parameterName=MobileShaderIntensityScale}, B=LinearInterpolate(A=@n56.RGB, B=@n57.RGB, Alpha=@n58))) {name=Shading, nodeColor=<rgba 1 0 0.7734 1>}
  NamedRerouteDeclaration(None=@n55) {name=SpriteWPOWithoutNearFade, nodeColor=<rgba 1 0 0.9141 1>}
 ]}
