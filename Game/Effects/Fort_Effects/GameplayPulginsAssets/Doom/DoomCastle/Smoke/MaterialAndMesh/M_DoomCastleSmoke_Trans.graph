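# Translucent, unlit material for the smoke effect on Doom Castle, built with material attributes.
# MF_DoomCloudBody and its mobile variant each feed emissive, opacity and world position offset, with the full version also driving opacity mask.
# PlatformSwitch_MA sends the full version to current-gen consoles and PC High to Low tiers, and the mobile one to Xb1/PS4, Switch, Android and PC-Perf.
# A Scalar_Time_of_Day call is present but unconnected.
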Material /Game/Effects/Fort_Effects/GameplayPulginsAssets/Doom/DoomCastle/Smoke/MaterialAndMesh/M_DoomCastleSmoke_Trans
graph {
 properties {
  bAutomaticallySetUsageInEditor=false
  bUseMaterialAttributes=true
  blendMode=BLEND_Translucent
  shadingModel=MSM_Unlit
  usedShadingModels=Unlit
 }
 nodes [
  MaterialFunctionCall() {materialFunction=&Scalar_Time_of_Day}
  n1 = MaterialFunctionCall() {materialFunction=&/MF_DoomCloudBody_Mobile}
  n2 = MakeMaterialAttributes!(
   BaseColor
   Metallic
   Specular
   Roughness
   Anisotropy
   EmissiveColor=@n1.Emissive
   Opacity=@n1.Emissive
   OpacityMask
   Normal
   Tangent
   WorldPositionOffset=@n1.Emissive
   SubsurfaceColor
   ClearCoat
   ClearCoatRoughness
   AmbientOcclusion
   Refraction
   CustomizedUVs_0
   CustomizedUVs_1
   CustomizedUVs_2
   CustomizedUVs_3
   CustomizedUVs_4
   CustomizedUVs_5
   CustomizedUVs_6
   CustomizedUVs_7
   PixelDepthOffset
   ShadingModel
   Displacement
  )
  n3 = MaterialFunctionCall() {materialFunction=&/MF_DoomCloudBody}
  n4 = MakeMaterialAttributes(EmissiveColor=@n3.Emissive, Opacity=@n3.Emissive, OpacityMask=@n3.Emissive, WorldPositionOffset=@n3.Emissive)
  MaterialFunctionCall!("Xb1,PS4"=@n2, "PC-High&&Epic,XSX,PS5"=@n4, "PC-Perf,Android"=@n2, Switch=@n2, "PC-Low(NotPerf)&&Medium"=@n4) {materialFunction=/Game/Effects/Niagara/MasterMaterials/Functions/Common/PlatformSwitch_MA}
 ]
}
