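# Unlit, two-sided translucent smoke material for Doom Castle spline trails, with a fake fog function and time-of-day tinting.
# Panning trail and ink textures, depth fade and tip edge masks build opacity, which fades near and far from the camera.
# Detail noise blends in a glow color that MPC HeadActive pushes toward a hue shift away from the actor.
# SplineThicken and CameraOffset drive world position offset, scaled down near the camera.
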Material /Game/Effects/Fort_Effects/GameplayPulginsAssets/Doom/DoomCastle/Smoke/MaterialAndMesh/M_DoomSmoke_Spline_Fog
graph {
 properties {
  bAutomaticallySetUsageInEditor=false
  blendMode=BLEND_Translucent
  pixelDepthOffsetMode=PDOM_Legacy
  shadingModel=MSM_Unlit
  twoSided=true
  usedShadingModels=Unlit
 }
 nodes [
  n1 = TextureCoordinate()
  n2 = MaterialFunctionCall(Float2=@n1) {materialFunction=&BreakOutFloat2Components}
  n3 = VectorParameter() {
   group=TipMask
   parameterName=TipMask_EdgeMask_InvertControl
   channelNames {a="", b=Power, g=Multiply, r=Offset}
   defaultValue=<rgba 0 500000 1 1>
  }
  n4 = VectorParameter() {group=TipMask, parameterName=TipMask_EdgeMask_Control, channelNames {a="", b=Power, g=Multiply, r=Offset}, defaultValue=<rgba 0 2 1.5 1>}
  n5 = TextureObjectParameter() {group=SmokeTrailTexture, parameterName=SmokeTrailTexture4_Texture, sortPriority=0, texture=/Game/Items/Traps/Textures/T_Limbo_Ink_01_M}
  n6 = VectorParameter() {
   group=SmokeTrailTexture
   parameterName=SmokeTrailTexture2_UVControl
   channelNames {a=VPan, b=UPan, g=VScale, r=UScale}
   defaultValue=<rgba 1 2 0 0.025>
  }
  n7 = TextureObjectParameter() {
   group=SmokeTrailTexture
   parameterName=SmokeTrailTexture_Texture
   samplerType=SAMPLERTYPE_Grayscale
   sortPriority=0
   texture=/Game/Effects/Fort_Effects/Textures/Smoke/T_Smoke_Trail_Soft_NoClamp
  }
  n8 = VectorParameter() {
   group=SmokeTrailTexture
   parameterName=SmokeTrailTexture_UVControl
   channelNames {a=VPan, b=UPan, g=VScale, r=UScale}
   defaultValue=<rgba 2 1 0.025 0>
  }
  n9 = MaterialFunctionCall(Float2=MaterialFunctionCall(XY=@n1) {materialFunction=&Swizzle}.YX) {materialFunction=&BreakOutFloat2Components}
  n10 = MaterialFunctionCall(X=Power(Base=@n9.R, Exp=ScalarParameter() {defaultValue=1.5, parameterName=ssf321}), Y=@n9.G) {materialFunction=&MakeFloat2}
  n11 = Add(
   A=MaterialFunctionCall(Texture=@n5, UVControl_RGB=@n6, UVControl_A=@n6) {materialFunction=&FXCommon_TextureSample}.G
   B=DepthFade(
    Opacity=MaterialFunctionCall(Texture=@n7, UVControl_RGB=@n8, UVControl_A=@n8, "[Override]UVInput"=@n10.Result) {materialFunction=&FXCommon_TextureSample}.R
    FadeDistance=ScalarParameter() {defaultValue=400, parameterName=FD}
   )
  )
  n12 = Multiply(
   A=Multiply(
    A=MaterialFunctionCall!(UorV=@n2.G, Invert_EdgeControl=@n3, NoInvert_EdgeControl=@n4) {materialFunction=/Game/Effects/Niagara/MasterMaterials/Functions/Common/FXCommon_EdgeMask_Ex}.Mask
    B=ScalarParameter() {defaultValue=0.7, parameterName=2D}
   )
   B=Power(Base=@n11, Exp=ScalarParameter() {defaultValue=1.5, parameterName=OpacityPower})
  )
  BlackBody!(Temp=Multiply(A=@n12, B=ScalarParameter() {defaultValue=1500, parameterName=K}))
  n13 = VectorParameter() {parameterName=C, defaultValue=<rgba 0.1 0.1 0.1 1>}
  n14 = MaterialFunctionCall(
   Color_Morning=VectorParameter() {parameterName=Morning, defaultValue=<rgba 1 0.6 0.6 1>}.RGB
   Color_Day=Constant3Vector() {constant=<rgba 1 1 1 0>}
   Color_Evening=VectorParameter() {parameterName=Evening, defaultValue=<rgba 0.8 0.52 0.52 1>}.RGB
   Color_Night=VectorParameter() {parameterName=Night, defaultValue=<rgba 0.3 0.6 1.2 1>}.RGB
  ) {materialFunction=&Color_Time_of_Day}
  n15 = VectorParameter() {parameterName=C2, defaultValue=<rgba 3 0.05 0 1>}
  n16 = VectorParameter() {parameterName=MaskGreen, defaultValue=<rgba 3000 0.6 0.3 0.75>}
  n17 = ActorPositionWS()
  n18 = WorldPosition()
  n19 = LinearInterpolate(
   A=@n15.RGB
   B=Multiply(
    A=MaterialFunctionCall(Hue Shift Percentage=@n16.B, Texture=@n15.RGB) {materialFunction=&HueShift}.Result
    B=@n16.A
   )
   Alpha=OneMinus(None=SphereMask(A=@n17, B=@n18.XYZ, Radius=@n16.R, Hardness=@n16.R))
  )
  n20 = CollectionParameter() {collection=/Game/Effects/Fort_Effects/GameplayPulginsAssets/Doom/DoomCastle/HeadFX/MPC/MPC_LodeSwanHead, parameterName=HeadActive}
  n21 = VectorParameter() {
   group=TipBur
   parameterName=TipBur_EdgeMask_InvertControl
   channelNames {a="", b=Power, g=Multiply, r=Offset}
   defaultValue=<rgba 0 55555 1 1>
  }
  n22 = VectorParameter() {group=TipBur, parameterName=TipBur_EdgeMask_Control, channelNames {a="", b=Power, g=Multiply, r=Offset}, defaultValue=<rgba 0 1 8 1>}
  n23 = Multiply(
   A=MaterialFunctionCall(UorV=@n2.G, Invert_EdgeControl=@n21, NoInvert_EdgeControl=@n22) {materialFunction=&FXCommon_EdgeMask_Ex}.Mask
   B=@n11
  )
  n24 = VectorParameter() {
   group=SmokeDetail
   parameterName=SmokeDetail_UVControl
   channelNames {a=VPan, b=UPan, g=VScale, r=UScale}
   defaultValue=<rgba 0.3 2 0.05 0.025>
  }
  n25 = Power(
   Base=MaterialFunctionCall(
    Texture=TextureObjectParameter() {group=SmokeDetail, parameterName=SmokeDetail_Texture, sortPriority=0, texture=/Game/Items/Traps/Textures/T_Limbo_Ink_02_M}
    UVControl_RGB=@n24
    UVControl_A=@n24
   ) {materialFunction=&FXCommon_TextureSample}.R
   Exp=ScalarParameter() {defaultValue=0.5, parameterName=f23}
  )
  n26 = Divide(A=LinearInterpolate(A=Multiply(A=@n13.RGB, B=@n14.Result), B=LinearInterpolate(A=@n15.RGB, B=@n19, Alpha=@n20), Alpha=Multiply(A=@n23, B=@n25))) {constB=1.97}
  MaterialFunctionCall!(ColorIn=@n26) {materialFunction=/Game/Athena/Helios/Maps/Test/Dev/Yi/MaterialAndMesh/MF_FakeFog}
  n27 = Distance(A=ActorPositionWS(), B=CameraPositionWS())
  n28 = ScalarParameter() {defaultValue=5000, parameterName=FadeDistance}
  n29 = Subtract(A=ScalarParameter() {defaultValue=200000, parameterName=StartFadeDistance}, B=@n28)
  n30 = Saturate(None=Divide(A=Subtract(A=@n27, B=@n29), B=@n28))
  n31 = Length(None=Subtract(A=WorldPosition().XYZ, B=CameraPositionWS())) {is=class MaterialExpressionLength}
  n32 = ScalarParameter() {defaultValue=500, parameterName=NFD_MinDist}
  n33 = Saturate(None=Divide(A=Subtract(A=@n31, B=@n32), B=ScalarParameter() {defaultValue=3500, parameterName=NFD_Distance}))
  Multiply(A=Multiply(A=OneMinus(None=@n30), B=@n12), B=@n33)
  n34 = ScalarParameter() {parameterName=f32}
  n35 = ScalarParameter() {defaultValue=5000, parameterName=ssssssW}
  n36 = VectorParameter() {
   group=TipSpilineWI
   parameterName=TipSpilineWI_EdgeMask_InvertControl
   channelNames {a="", b=Power, g=Multiply, r=Offset}
   defaultValue=<rgba 1 100000 1 1>
  }
  n37 = VectorParameter() {
   group=TipSpilineWI
   parameterName=TipSpilineWI_EdgeMask_Control
   channelNames {a="", b=Power, g=Multiply, r=Offset}
   defaultValue=<rgba 0 1 1 1>
  }
  n38 = MaterialFunctionCall(UorV=@n2.G, Invert_EdgeControl=@n36, NoInvert_EdgeControl=@n37) {materialFunction=&FXCommon_EdgeMask_Ex}
  n39 = LinearInterpolate(A=@n35, B=ScalarParameter() {defaultValue=1000, parameterName=ssssssW2}, Alpha=@n38.Mask)
  Multiply(
   A=Add(
    A=MaterialFunctionCall(Offset Amount=@n34) {materialFunction=/Engine/Functions/Engine_MaterialFunctions02/WorldPositionOffset/CameraOffset}.Camera Offset
    B=MaterialFunctionCall(WidthBase=@n39, WidthTip=@n39) {materialFunction=&SplineThicken}.WorldPositionOffset
   )
   B=Power(Base=@n33) {constExponent=0.5}
  )
  Multiply(A=@n35, B=@n38.Mask)
  Reroute(None=MaterialFunctionCall(Emissive=@n26) {materialFunction=&MF_FX_EmissiveAdjustments}.Result)
 ]
}
