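# Translucent, two-sided lit material for the smoky forge cap effect in Doom Castle, with red glow and dark smoke tint.
# A rotated, panning smoke texture drives color and opacity, while a random texture sampled at a fixed UV and panned over time flickers the glow.
# Opacity combines depth, camera-distance and Fresnel fades with UV edge and seam masks.
# A platform switch lowers opacity on Switch, Android and PC performance mode.
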Material /Game/Effects/Fort_Effects/GameplayPulginsAssets/Doom/DoomCastle/Smoke/MaterialAndMesh/M_ForgeCap
graph {
 properties {
  bAutomaticallySetUsageInEditor=false
  blendMode=BLEND_Translucent
  translucencyPass=MTP_BeforeDOF
  twoSided=true
  usedShadingModels=Default Lit
 }
 nodes [
  n1 = MaterialFunctionCall(Float2=TextureCoordinate()) {materialFunction=&BreakOutFloat2Components}
  n2 = Power(Base=@n1.R, Exp=ScalarParameter() {defaultValue=0.5, parameterName=s2}) {constExponent=50}
  n3 = VectorParameter() {parameterName=CapLightnColor, defaultValue=<rgba 35 0 0 5>}
  n4 = Power(Base=@n1.R, Exp=@n3.A)
  n5 = TextureObjectParameter() {
   group=Random
   parameterName=Random_Texture
   sortPriority=0
   texture=/Game/Environments/Asteria/Sets/NeonCity/Techart/Screen/Textures/Random_M
  }
  n6 = VectorParameter() {group=Random, parameterName=Random_UVControl, channelNames {a=VPan, b=UPan, g=VScale, r=UScale}, defaultValue=<rgba 1 1 1 0>}
  n7 = MaterialFunctionCall(
   Texture=@n5
   UVControl_RGB=@n6
   UVControl_A=@n6
   "[Override]UVInput"=MaterialFunctionCall(R=Constant(), G=Constant() {r=0.5}) {materialFunction=&AppendMany}.RG
  ) {materialFunction=&FXCommon_TextureSample}
  n8 = Constant() {r=1}
  n9 = TextureCoordinate()
  n10 = MaterialFunctionCall(Float2=@n9) {materialFunction=&BreakOutFloat2Components}
  n11 = VectorParameter() {
   group=SeamRemove
   parameterName=SeamRemove_EdgeMask_Control
   channelNames {a="", b=Power, g=Multiply, r=Offset}
   defaultValue=<rgba 0 15 1.3 1>
  }
  n12 = MaterialFunctionCall!(UorV=@n10.G, EdgeControl=@n11) {materialFunction=/Game/Effects/Niagara/MasterMaterials/Functions/Common/FXCommon_EdgeMask_Simple}
  n13 = Multiply(
   A=MaterialFunctionCall(
    Input=NamedRerouteDeclaration(None=@n7.R) {name=RandomV, nodeColor=<rgba 1 0.6562 0 1>}
    Input Low=Constant()
    Input High=@n8
    Target Low=ScalarParameter() {defaultValue=0.3, parameterName=TargetL}
    Target High=@n8
   ) {materialFunction=&RemapValueRange}.Result
   B=Multiply(A=@n3.RGB, B=@n12.Mask)
  )
  n14 = TextureObjectParameter() {
   group=FlowUpTexture
   parameterName=FlowUpTexture_Texture
   sortPriority=0
   texture=&/Smoke01_D
  }
  n15 = VectorParameter() {
   group=FlowUpTexture
   parameterName=FlowUpTexture_UVControl
   channelNames {a=VPan, b=UPan, g=VScale, r=UScale}
   defaultValue=<rgba 0.4 1 -0.03 0>
  }
  n16 = MaterialFunctionCall!(
   Texture=@n14
   AdditionalUVOffset
   UVControl_RGB=@n15
   UVControl_A=@n15
   "[Override]TimeInput"
   "[Override]ScaleUVByCenter"=StaticBool() {value=true}
   "[Override]UVInput"=Rotator(Coordinate=@n9, Time=ScalarParameter() {defaultValue=-2, parameterName=2d1})
  ) {materialFunction=/Game/Effects/Niagara/MasterMaterials/Functions/Common/FXCommon_TextureSample_V2}
  n17 = OneMinus(None=@n16.R)
  n18 = Multiply(
   A=Multiply(A=OneMinus(None=@n2), B=Multiply(A=@n4, B=@n13))
   B=Power(Base=@n17, Exp=ScalarParameter() {defaultValue=15, parameterName=SmokeTextureEXP})
  )
  n19 = ScalarParameter() {parameterName=d131}
  n20 = Multiply(
   A=LinearInterpolate(B=Power(Base=@n16.R, Exp=@n19), Alpha=@n12.Mask) {constA=1}
   B=VectorParameter() {parameterName=SmokeColor, defaultValue=<rgba 0.07 0.0356 0.035 1>}.RGB
  )
  Add(A=@n18, B=@n20)
  n21 = MaterialFunctionCall(Float2=TextureCoordinate()) {materialFunction=&BreakOutFloat2Components}
  n22 = OneMinus(None=Power(Base=OneMinus(None=@n21.R), Exp=ScalarParameter() {defaultValue=1, parameterName=PowerDither}) {constExponent=3})
  MaterialFunctionCall(Alpha Threshold=@n22) {materialFunction=&DitherTemporalAA}
  MaterialFunctionCall(Emissive=@n20) {materialFunction=&MF_FX_EmissiveAdjustments}
  MaterialFunctionCall(Emissive=@n18) {materialFunction=&MF_FX_EmissiveAdjustments}
  n23 = Distance(A=ActorPositionWS(), B=CameraPositionWS())
  n24 = VectorParameter() {parameterName=CameraNearFade, defaultValue=<rgba 500 1500 0 1>}
  n25 = Saturate(None=Divide(A=Subtract(A=@n23, B=@n24.R), B=@n24.G))
  n26 = LinearInterpolate(A=ScalarParameter() {defaultValue=8, parameterName="WF##"}, B=ScalarParameter() {defaultValue=2.2, parameterName="WF##2"}, Alpha=@n22)
  n27 = Saturate(None=Multiply(A=Power(Base=@n17, Exp=@n26), B=ScalarParameter() {defaultValue=3, parameterName=23}))
  n28 = Multiply(A=@n27, B=ScalarParameter() {defaultValue=0.95, parameterName=21})
  n29 = DepthFade(
   Opacity=Add(A=Saturate(None=@n28), B=ScalarParameter() {defaultValue=0.2, parameterName=213f2})
   FadeDistance=ScalarParameter() {defaultValue=150, parameterName=sDF}
  )
  n30 = MaterialFunctionCall(Float2=TextureCoordinate()) {materialFunction=&BreakOutFloat2Components}
  n31 = VectorParameter() {
   group=EdgeMask
   parameterName=EdgeMask_EdgeMask_InvertControl
   channelNames {a="", b=Power, g=Multiply, r=Offset}
   defaultValue=<rgba 0 4.5 1 1>
  }
  n32 = VectorParameter() {group=EdgeMask, parameterName=EdgeMask_EdgeMask_Control, channelNames {a="", b=Power, g=Multiply, r=Offset}, defaultValue=<rgba 0 3.5 2 1>}
  n33 = Constant() {r=1}
  n34 = ScalarParameter() {defaultValue=0.2, parameterName=ScaleOpacityMobile}
  n35 = Multiply(
   A=Multiply(
    A=Multiply(A=@n25, B=@n29)
    B=MaterialFunctionCall!(UorV=@n30.R, Invert_EdgeControl=@n31, NoInvert_EdgeControl=@n32) {materialFunction=/Game/Effects/Niagara/MasterMaterials/Functions/Common/FXCommon_EdgeMask_Ex}.Mask
   )
   B=MaterialFunctionCall("Xb1,PS4"=@n33, "PC-High&&Epic,XSX,PS5"=@n33, "PC-Perf,Android"=@n34, Switch=@n34, "PC-Low(NotPerf)&&Medium"=@n33) {materialFunction=&PlatformSwitch_1D}.Result
  )
  n36 = MaterialFunctionCall(Distance=ScalarParameter() {defaultValue=2000, parameterName=NearFadeDistance}) {materialFunction=&Near_Camera_Fade}
  n37 = ScalarParameter() {defaultValue=5, parameterName=Fresn}
  n38 = Constant()
  n39 = Power(Base=OneMinus(None=Fresnel(ExponentIn=@n37, BaseReflectFractionIn=@n38)), Exp=ScalarParameter() {defaultValue=3, parameterName=FrePower})
  n40 = Power(Base=MaterialFunctionCall(Distance=ScalarParameter() {defaultValue=3000, parameterName=FreDistance}) {materialFunction=&Near_Camera_Fade}.Result) {constExponent=20}
  n41 = VectorParameter() {
   group=SeamRemove
   parameterName=SeamRemove_EdgeMask_Control2
   channelNames {a="", b=Power, g=Multiply, r=Offset}
   defaultValue=<rgba 0 15 1 1>
  }
  Multiply(
   A=Multiply(A=Multiply(A=@n35, B=@n36.Result), B=LinearInterpolate(A=@n39, Alpha=@n40) {constA=1})
   B=MaterialFunctionCall(UorV=@n10.G, EdgeControl=@n41) {materialFunction=&FXCommon_EdgeMask_Simple}.Mask
  )
  Multiply(A=@n27, B=@n22)
 ]
}
