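# Unlit translucent, depth-test-free smoke overlay for Doom Castle, showing low-altitude cloud haze when the camera is within a volume.
# Camera direction and position drive a sliced 3D noise texture for opacity and curve-atlas color, with an optional SandStorm tint.
# Custom primitive data defines box and cylinder fades from camera position, plus a large downward offset when outside the box.
# A red forge glow fades with camera distance from an offset actor point.
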Material /Game/Effects/Fort_Effects/GameplayPulginsAssets/Doom/DoomCastle/Smoke/MaterialAndMesh/M_LowAltitudeClouds_DoomSmoke_PostProcess
graph {
 properties {
  bDisableDepthTest=true
  bUseTranslucencyVertexFog=false
  bUsedWithWater=true
  blendMode=BLEND_Translucent
  blendableLocation=BL_SceneColorAfterDOF
  pixelDepthOffsetMode=PDOM_Legacy
  shadingModel=MSM_Unlit
  usedShadingModels=Unlit
 }
 nodes [
  n1 = CameraVectorWS()
  n2 = CameraPositionWS()
  n3 = MaterialFunctionCall!(
   UVs=Subtract(A=Multiply(A=@n1) {constB=0.18}, B=Multiply(A=@n2) {constB=0.0004})
   Slices Texture=TextureObject() {samplerType=SAMPLERTYPE_LinearColor, texture=/Game/Environments/Helios/TechArt/LowAltitudeClouds/T_LowAltitudeClouds_N}
   Atlas Size=Constant() {r=8}
   Atlas Padding=ScalarParameter() {defaultValue=0.9375, parameterName=AtlasPadding, primitiveDataIndex=8}
  ) {materialFunction=/Game/Environments/Helios/TechArt/LowAltitudeClouds/MF_3DVolumeFrom2DSlices}
  n4 = ComponentMask(None=Power(Base=@n3.A, Exp=ScalarParameter() {defaultValue=0.01, parameterName=F})) {b=true}
  n5 = Subtract(
   A=CameraPositionWS()
   B=MaterialFunctionCall() {materialFunction=&ObjectPivotPoint}.Object Pivot Location
  )
  n6 = VectorParameter() {bUseCustomPrimitiveData=true, parameterName=AxisX, primitiveDataIndex=4, defaultValue=<rgba 0 0 0 1>}
  n7 = ComponentMask(None=@n6.RGB) {b=true, g=true, r=true}
  n8 = VectorParameter() {bUseCustomPrimitiveData=true, parameterName=Scale, defaultValue=<rgba 0 0 0 1>}
  n9 = Subtract(A=Abs(None=DotProduct(A=@n5, B=@n7)), B=ComponentMask(None=@n8.RGB) {r=true})
  n10 = VectorParameter() {bUseCustomPrimitiveData=true, parameterName=AxisY, primitiveDataIndex=8, defaultValue=<rgba 0 0 0 1>}
  n11 = ComponentMask(None=@n10.RGB) {b=true, g=true, r=true}
  n12 = Subtract(A=Abs(None=DotProduct(A=@n5, B=@n11)), B=ComponentMask(None=@n8.RGB) {g=true})
  n13 = VectorParameter() {bUseCustomPrimitiveData=true, parameterName=AxisZ, primitiveDataIndex=12, defaultValue=<rgba 0 0 0 1>}
  n14 = Abs(None=DotProduct(A=@n5, B=ComponentMask(None=@n13.RGB) {b=true, g=true, r=true}))
  n15 = ComponentMask(None=@n8.RGB) {b=true}
  n16 = Min(A=Min(A=OneMinus(None=@n9), B=OneMinus(None=@n12)), B=OneMinus(None=Subtract(A=@n14, B=@n15)))
  n17 = ScalarParameter() {bUseCustomPrimitiveData=true, defaultValue=550, parameterName=Fade, primitiveDataIndex=20}
  n18 = Divide(A=@n16, B=@n17)
  n19 = Multiply(A=Multiply(A=@n4, B=@n4), B=VertexInterpolator(VS=Saturate(None=@n18)).PS)
  n20 = Constant() {r=0.5}
  n21 = TextureObjectParameter() {
   parameterName=TOD Curves
   samplerType=SAMPLERTYPE_LinearColor
   texture=&/CA_LowAltitudeClouds
  }
  n22 = TextureSample(UVs=AppendVector(A=@n20, B=Constant()), Tex=@n21) {
   samplerSource=SSM_Clamp_WorldGroupSettings
   samplerType=SAMPLERTYPE_LinearColor
   texture=/Game/Environments/Helios/TechArt/LowAltitudeClouds/CA_LowAltitudeClouds
  }
  n23 = TextureSample(UVs=AppendVector(A=@n20, B=Constant() {r=1}), Tex=@n21) {
   samplerSource=SSM_Clamp_WorldGroupSettings
   samplerType=SAMPLERTYPE_LinearColor
   texture=&CA_LowAltitudeClouds
  }
  n24 = OneMinus(None=ComponentMask(None=@n3.RGB) {b=true})
  n25 = Multiply(A=Add(A=Multiply(A=@n24, B=@n24)) {constB=-0.5}) {constB=2}
  n26 = Max(A=Add(A=Multiply(A=@n25) {constB=0.5}) {constB=0.7}) {constB=0}
  n27 = Multiply(A=@n26, B=@n26)
  n28 = VectorParameter() {bUseCustomPrimitiveData=true, parameterName=Color Multiplier, primitiveDataIndex=16, defaultValue=<rgba 0 0 0 1>}
  n29 = VectorParameter() {parameterName=ForgeColor, defaultValue=<rgba 0.15 0 0 1>}
  n30 = Normalize(VectorInput=Transform(None=Constant3Vector() {constant=<rgba 0 0 1 0>}) {transformSourceType=TRANSFORMSOURCE_Local})
  n31 = VectorParameter() {parameterName=ForgeLightCenter_Offset, defaultValue=<rgba 0 0 550 1>}
  n32 = Normalize(VectorInput=Transform(None=@n31.RGB) {transformSourceType=TRANSFORMSOURCE_Local})
  n33 = Length(None=@n31.RGB) {is=class MaterialExpressionLength}
  n34 = Subtract(A=Add(A=ActorPositionWS(), B=Multiply(A=@n32, B=@n33)), B=CameraPositionWS())
  n35 = Length(None=MaterialFunctionCall!(Plane Normal=@n30, Vector=@n34) {materialFunction=/Engine/Functions/Engine_MaterialFunctions02/Math/ProjectVectorOntoPlane}.Result)
  n36 = VectorParameter() {parameterName=ForgeLight, defaultValue=<rgba 1000 1500 500 5000>}
  n37 = Saturate(None=Divide(A=Subtract(A=@n35, B=@n36.R), B=@n36.G))
  n38 = Length(None=@n34)
  n39 = Saturate(None=Divide(A=Subtract(A=@n38, B=@n36.B), B=@n36.A))
  n40 = Multiply(A=Power(Base=@n27) {constExponent=0.1}, B=Multiply(A=OneMinus(None=@n37), B=OneMinus(None=@n39)))
  n41 = Multiply(
   A=LinearInterpolate(A=VertexInterpolator(VS=@n22.RGB).PS, B=VertexInterpolator(VS=@n23.RGB).PS, Alpha=@n27)
   B=LinearInterpolate(A=@n28.RGB, B=Add(A=@n28.RGB, B=@n29.RGB), Alpha=Multiply(A=@n40))
  )
  n42 = CollectionParameter() {collection=/Game/Environments/Helios/TechArt/LowAltitudeClouds/SandStorm/MPC_LowAltitudeClouds_SandStorm, parameterName=Color Multiplier}
  n43 = Multiply(A=@n41, B=ComponentMask(None=@n42) {b=true, g=true, r=true})
  Add(
   A=MaterialFunctionCall(Number=@n19, UVs=ScreenPosition().ViewportUV) {materialFunction=&DebugScalarValues}.Result
   B=MaterialFunctionCall(Emissive=StaticSwitchParameter(True=@n43, False=@n41) {parameterName=SandStorm}) {materialFunction=&MF_FX_EmissiveAdjustments}.Result
  )
  CollectionParameter() {collection=&FortniteMaterialParameters, parameterName=TimeOfDay_0-1}
  Multiply(A=Saturate(None=@n19)) {constB=0.9}
  n44 = Ceil(None=@n16)
  Multiply(A=OneMinus(None=Saturate(None=@n44)), B=Constant3Vector() {constant=<rgba 0 0 -500000 0>})
  n45 = Length(None=MaterialFunctionCall(Plane Normal=@n30, Vector=@n5) {materialFunction=&ProjectVectorOntoPlane}.Result)
  n46 = ScalarParameter() {bUseCustomPrimitiveData=true, defaultValue=550, parameterName=CylinderFade, primitiveDataIndex=21}
  n47 = Saturate(None=Divide(A=Subtract(A=@n45, B=@n46), B=@n17))
  Multiply(A=@n19, B=VertexInterpolator(VS=OneMinus(None=@n47)).PS)
  Multiply()
 ]
}
