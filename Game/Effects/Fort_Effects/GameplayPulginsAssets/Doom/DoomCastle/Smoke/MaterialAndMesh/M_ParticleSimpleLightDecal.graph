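# Unlit translucent Niagara sprite and mesh particle material that fakes a light decal on scene geometry for Doom Castle effects.
# Particle color drives emissive; a sphere mask of scene position behind translucency around the particle forms the glow, with dynamic radius and falloff power.
# The mask only shows where scene-texture Metallic is above zero, so it is zero on forward and mobile paths.
# It fades in with horizontal camera distance, hiding up close.
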Material /Game/Effects/Fort_Effects/GameplayPulginsAssets/Doom/DoomCastle/Smoke/MaterialAndMesh/M_ParticleSimpleLightDecal
graph {
 properties {
  bAutomaticallySetUsageInEditor=false
  bUsedWithNiagaraMeshParticles=true
  bUsedWithNiagaraSprites=true
  blendMode=BLEND_Translucent
  shadingModel=MSM_Unlit
  usedShadingModels=Unlit
 }
 nodes [
  MaterialFunctionCall(Emissive=ParticleColor().RGB) {materialFunction=&MF_FX_EmissiveAdjustments}
  n1 = Constant3Vector() {constant=<rgba 0 0 1 0>}
  n2 = ParticlePositionWS()
  n3 = Subtract(A=CameraPositionWS(), B=@n2)
  n4 = VectorParameter() {parameterName=D, defaultValue=<rgba 2500 3000 0 1>}
  n5 = Subtract(
   A=Length(None=MaterialFunctionCall!(Plane Normal=@n1, Vector=@n3) {materialFunction=/Engine/Functions/Engine_MaterialFunctions02/Math/ProjectVectorOntoPlane}.Result) {is=class MaterialExpressionLength}
   B=@n4.R
  )
  n6 = MaterialFunctionCall(Float3=SceneTexture() {sceneTextureId=PPI_Metallic}.Color) {materialFunction=&BreakOutFloat3Components}
  n7 = Constant()
  n8 = ShadingPathSwitch(Default=@n6.R, Deferred=@n6.R, Forward=@n7, Mobile=@n7)
  n9 = DynamicParameter()
  n10 = SphereMask(
   A=MaterialFunctionCall() {materialFunction=&WorldPositionBehindTranslucency}.Result
   B=@n2
   Radius=@n9.R
  ) {hardnessPercent=0}
  Multiply(A=Saturate(None=Divide(A=@n5, B=@n4.G)), B=Multiply(A=Ceil(None=@n8), B=Power(Base=@n10, Exp=@n9.G)))
 ]
}
